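# Translucent unlit material for the black hole ring effect on the SuperNovaTaro backpack, built for Niagara mesh particles.
# Radial UVs, distorted by panning blurred noise, turn scrolling noise layers into a pink swirl over a Fresnel gradient, with faint screen-aligned nebula.
# A sphere mask blacks out the center, and particle color tints the glow.
# Swirling plasma noise drives opacity, scaled by particle alpha and depth fade, which some lower-end platforms skip.
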Material /Game/Accessories/FORT_Backpacks/Backpack_F_MED_SuperNovaTaro/FX/M_Backpack_SuperNovaTaro_Blackhole_Ring
graph {
 properties {
  bUsedWithNiagaraMeshParticles=true
  blendMode=BLEND_Translucent
  pixelDepthOffsetMode=PDOM_Legacy
  shadingModel=MSM_Unlit
  usedShadingModels=Unlit
 }
 nodes [
  n1 = TextureCoordinate()
  n2 = Multiply(A=TextureSample(UVs=Panner(Coordinate=@n1) {speedY=-0.2}) {texture=/Game/Characters/Player/Male/Medium/Bodies/M_Med_Soldier_04/Skins/Galaxy/Textures/T_FN_Nebula_style2}.RGB) {constB=5}
  n3 = Constant() {r=100}
  n4 = TextureSample(UVs=Panner(Coordinate=TextureCoordinate()) {speedY=-0.2}) {samplerType=SAMPLERTYPE_LinearGrayscale, texture=/Game/Effects/Fort_Effects/Emotes/UltraLight/Texture/T_DirectionalNoise_Greyscale}
  n5 = TextureSample(UVs=Add(
    A=MaterialFunctionCall!(TextureResolution=@n3) {materialFunction=/Engine/Functions/Engine_MaterialFunctions02/Texturing/ScreenAlignedPixelToPixelUVs}.Result
    B=Multiply(A=@n4.R) {constB=0.3}
   )) {texture=&T_FN_Nebula_style2}
  n6 = TextureCoordinate() {coordinateIndex=1}
  n7 = Constant() {r=0.5}
  n8 = Constant() {r=0.4}
  n9 = Multiply(
   A=Desaturation(None=Multiply(A=@n5.RGB) {constB=10}, Fraction=Constant() {r=-1})
   B=OneMinus(None=SphereMask(A=@n6, B=@n7, Radius=@n8) {hardnessPercent=50})
  )
  n10 = Constant3Vector() {constant=<rgba 0.3333 0 0.0903 1>}
  n11 = Multiply(A=Power(Base=Fresnel() {baseReflectFraction=0, exponent=1}) {constExponent=0.5})
  n12 = Multiply(A=LinearInterpolate(A=Add(A=@n10) {constB=0}, B=Constant3Vector() {constant=<rgba 0.0097 0 0.1823 1>}, Alpha=Saturate(None=@n11)))
  n13 = OneMinus(None=SphereMask(A=TextureCoordinate() {coordinateIndex=1}, B=Constant() {r=0.5}, Radius=Constant() {r=0.5}) {hardnessPercent=50})
  n14 = Subtract(A=TextureCoordinate() {coordinateIndex=1, uTiling=2, vTiling=2}, B=Constant() {r=1})
  n15 = MaterialFunctionCall(Vector or UVs=Subtract(A=TextureCoordinate() {coordinateIndex=1, uTiling=2, vTiling=2}, B=Constant() {r=1})) {materialFunction=&VectorToRadialValue}
  n16 = Constant2Vector() {g=1, r=1}
  n17 = TextureCoordinate() {coordinateIndex=1}
  n18 = TextureSample!(
   UVs=Panner(Coordinate=Multiply(A=@n15.Radial Coordinates, B=@n16)) {speedX=-0.2}
   Tex
   "DDX(UVs)"=DDX(Value=@n17)
   "DDY(UVs)"=DDY(Value=@n17)
   Apply View MipBias
  ) {
   mipValueMode=TMVM_Derivative
   samplerType=SAMPLERTYPE_LinearGrayscale
   texture=/Engine/Functions/Engine_MaterialFunctions02/ExampleContent/Textures/LowResBlurredNoise
  }
  n19 = Add(
   A=Multiply(
    A=Constant2Vector() {g=2, r=1}
    B=MaterialFunctionCall(Vector or UVs=@n14) {materialFunction=&VectorToRadialValue}.Radial Coordinates
   )
   B=Multiply(A=@n18.R) {constB=0.2}
  )
  n20 = TextureCoordinate() {coordinateIndex=1}
  n21 = Multiply(A=TextureSample(UVs=Panner(Coordinate=@n19) {speedX=0.2}, "DDX(UVs)"=DDX(Value=@n20), "DDY(UVs)"=DDY(Value=@n20)) {mipValueMode=TMVM_Derivative, texture=/Game/Effects/Fort_Effects/Textures/Tile/T_Noise_02}.R) {constB=30}
  n22 = TextureCoordinate() {coordinateIndex=1}
  n23 = Constant() {r=0.5}
  n24 = Constant() {r=0.5}
  n25 = TextureCoordinate() {coordinateIndex=1, uTiling=2, vTiling=2}
  n26 = Constant() {r=1}
  n27 = Multiply(
   A=Constant2Vector() {g=1, r=1}
   B=MaterialFunctionCall(Vector or UVs=Subtract(A=@n25, B=@n26)) {materialFunction=&VectorToRadialValue}.Radial Coordinates
  )
  n28 = MaterialFunctionCall(Vector or UVs=Subtract(A=TextureCoordinate() {coordinateIndex=1, uTiling=2, vTiling=2}, B=Constant() {r=1})) {materialFunction=&VectorToRadialValue}
  n29 = Constant2Vector() {g=1, r=1}
  n30 = TextureCoordinate() {coordinateIndex=1}
  n31 = TextureSample(UVs=Panner(Coordinate=Multiply(A=@n28.Radial Coordinates, B=@n29)) {speedX=-0.2}, "DDX(UVs)"=DDX(Value=@n30), "DDY(UVs)"=DDY(Value=@n30)) {
   mipValueMode=TMVM_Derivative
   samplerType=SAMPLERTYPE_LinearGrayscale
   texture=&LowResBlurredNoise
  }
  n32 = Multiply(A=@n31.R) {constB=0.2}
  n33 = TextureCoordinate() {coordinateIndex=1}
  n34 = TextureSample(UVs=Panner(Coordinate=Add(A=@n27, B=@n32)) {speedX=0.2}, "DDX(UVs)"=DDX(Value=@n33), "DDY(UVs)"=DDY(Value=@n33)) {
   mipValueMode=TMVM_Derivative
   samplerType=SAMPLERTYPE_LinearGrayscale
   texture=/Game/Gadgets/Assets/VinderTech_GliderChute/Glider_Dazzle/FX/Variant/T_Glider_Dazzle_Electric_Crystal_Noise
  }
  n35 = Clamp(None=Power(Base=Multiply(A=@n34.R) {constB=3}) {constExponent=1}) {maxDefault=5}
  n36 = LinearInterpolate(
   A=Multiply(A=Power(Base=@n21) {constExponent=3}, B=SphereMask(A=@n22, B=@n23, Radius=@n24) {hardnessPercent=0})
   B=Multiply(A=Add(A=@n35) {constB=0}) {constB=5}
  ) {constAlpha=0.3}
  n37 = Multiply(
   A=Constant3Vector() {constant=<rgba 1 0.1009 0.5034 1>}
   B=Multiply(A=Multiply(B=@n36) {constA=1}, B=LinearInterpolate(Alpha=@n13) {constB=0.5})
  )
  n38 = ParticleColor()
  n39 = Multiply(A=LinearInterpolate(A=LinearInterpolate(A=@n2, B=@n9) {constAlpha=1}, B=Add(A=@n12, B=@n37) {constB=0}) {constAlpha=0.95}, B=@n38.RGB)
  n40 = TextureCoordinate() {coordinateIndex=1}
  n41 = Constant() {r=0.5}
  MaterialFunctionCall(Emissive=LinearInterpolate(A=Add(B=@n39), Alpha=SphereMask(A=@n40, B=@n41) {attenuationRadius=0.2, hardnessPercent=50}) {constB=0}) {materialFunction=&MF_FX_EmissiveAdjustments}
  n42 = TextureCoordinate() {coordinateIndex=1, uTiling=2, vTiling=2}
  n43 = Constant() {r=1}
  n44 = Multiply(
   A=Constant2Vector() {g=1, r=1}
   B=MaterialFunctionCall(Vector or UVs=Subtract(A=@n42, B=@n43)) {materialFunction=&VectorToRadialValue}.Radial Coordinates
  )
  n45 = MaterialFunctionCall(Vector or UVs=Subtract(A=TextureCoordinate() {coordinateIndex=1, uTiling=2, vTiling=2}, B=Constant() {r=1})) {materialFunction=&VectorToRadialValue}
  n46 = Constant2Vector() {g=1, r=1}
  n47 = TextureCoordinate() {coordinateIndex=1}
  n48 = TextureSample(UVs=Panner(Coordinate=Multiply(A=@n45.Radial Coordinates, B=@n46)) {speedX=-0.1}, "DDX(UVs)"=DDX(Value=@n47), "DDY(UVs)"=DDY(Value=@n47)) {
   mipValueMode=TMVM_Derivative
   samplerType=SAMPLERTYPE_LinearGrayscale
   texture=&LowResBlurredNoise
  }
  n49 = Multiply(A=@n48.R) {constB=0.2}
  n50 = TextureCoordinate() {coordinateIndex=1}
  n51 = TextureSample(UVs=Panner(Coordinate=Add(A=@n44, B=@n49)) {speedX=0.2}, "DDX(UVs)"=DDX(Value=@n50), "DDY(UVs)"=DDY(Value=@n50)) {mipValueMode=TMVM_Derivative, samplerType=SAMPLERTYPE_LinearColor, texture=/Game/Effects/Fort_Effects/Emotes/UltraLight/Texture/T_Plasma_D_C}
  n52 = LinearInterpolate(Alpha=SphereMask(A=TextureCoordinate() {coordinateIndex=1}, B=Constant() {r=0.5}, Radius=Constant() {r=0.5}) {hardnessPercent=0}) {constA=3, constB=0}
  n53 = Multiply(A=Multiply(A=Power(Base=@n51.R, Exp=@n52)) {constB=2}) {constB=4}
  n54 = DepthFade(FadeDistance=ScalarParameter() {parameterName=DepthFade})
  n55 = Constant() {r=1}
  Multiply(
   A=@n38.A
   B=Multiply(
    A=Saturate(None=@n53)
    B=MaterialFunctionCall("Xb1,PS4"=@n54, "PC-High&&Epic,XSX,PS5"=@n54, "PC-Perf,Android"=@n55, Switch=@n55, "PC-Low(NotPerf)&&Medium"=@n55) {materialFunction=&PlatformSwitch_1D}.Result
   )
  )
 ]
}
